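# Bonus-goals quest reward panel widget for the Sparks Pass S05 S04 event, showing four reward images over background and foreground glow materials.
# On Construct it starts the FloatReward1-4 animations with endless looping, each bobbing and placing one image through render transform keys.
# A combined FloatAnim and an empty Default animation exist but are never played here.
# The bottom "Axion Sentinel" quest info border and rich text are set to hidden.
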WidgetBlueprint /SparksPassS05/Items/QuestRewards/BonusGoals/WBP_Sparks_QuestRewards_Panel_Bonus_S04Event
animations [
 Default {
  binding_count=0
  length_seconds=5
  object=&self:Default
  sequence {bindings []}
 }
 FloatAnim {
  binding_count=4
  length_seconds=1.75
  object=&self:FloatAnim
  sequence {bindings [
    Image_Reward4 {
     id=#1
     possessed_class=Image
     tracks [{
       type=MovieScene2DTransformTrack
       sections [{
         type=MovieScene2DTransformSection
         end_frame=35
         start_frame=0
         channels [
          Translation.X {type=MovieSceneScriptingFloatChannel, keys [[0,78],[35,78],[35.0003,241]]}
          Translation.Y {type=MovieSceneScriptingFloatChannel, keys [[0,-95],[17,-109],[35,-95],[35.0003,0]]}
          Angle {type=MovieSceneScriptingFloatChannel, const [3,1]}
          Scale.X {type=MovieSceneScriptingFloatChannel, keys [[0,1.3],[35,1.3],[35.0003,2]]}
          Scale.Y {type=MovieSceneScriptingFloatChannel, keys [[0,1.3],[35,1.3],[35.0003,2]]}
          Shear.X {type=MovieSceneScriptingFloatChannel, const [0,1]}
          Shear.Y {type=MovieSceneScriptingFloatChannel, const [0,1]}
         ]
        }]
      }]
    }
    Image_Reward3 {
     id=#2
     possessed_class=Image
     tracks [{
       type=MovieScene2DTransformTrack
       sections [{
         type=MovieScene2DTransformSection
         end_frame=35
         start_frame=-1
         channels [
          Translation.X {type=MovieSceneScriptingFloatChannel, keys [[-0.0003,-307],[0,-290]]}
          Translation.Y {type=MovieSceneScriptingFloatChannel, keys [[-0.0003,-154],[12.9803,-168],[35,-154]]}
          Angle {type=MovieSceneScriptingFloatChannel, const [-5,1]}
          Scale.X {type=MovieSceneScriptingFloatChannel, const [1.5,2]}
          Scale.Y {type=MovieSceneScriptingFloatChannel, const [1.5,2]}
          Shear.X {type=MovieSceneScriptingFloatChannel, const [0,1]}
          Shear.Y {type=MovieSceneScriptingFloatChannel, const [0,1]}
         ]
        }]
      }]
    }
    Image_Reward2 {
     id=#3
     possessed_class=Image
     tracks [{
       type=MovieScene2DTransformTrack
       sections [{
         type=MovieScene2DTransformSection
         end_frame=35
         start_frame=-1
         channels [
          Translation.X {type=MovieSceneScriptingFloatChannel, keys [[-0.0003,-23.7008],[0,-100]]}
          Translation.Y {type=MovieSceneScriptingFloatChannel, keys [[-0.0003,26],[0,168],[17.0217,154],[35,168]]}
          Angle {type=MovieSceneScriptingFloatChannel, const [5,1]}
          Scale.X {type=MovieSceneScriptingFloatChannel, keys [[-0.0003,2.7],[0,1.5],[35,1.5]]}
          Scale.Y {type=MovieSceneScriptingFloatChannel, keys [[-0.0003,2.7],[0,1.5],[35,1.5]]}
          Shear.X {type=MovieSceneScriptingFloatChannel, const [0,1]}
          Shear.Y {type=MovieSceneScriptingFloatChannel, const [0,1]}
         ]
        }]
      }]
    }
    Image_Reward1 {
     id=#4
     possessed_class=Image
     tracks [{
       type=MovieScene2DTransformTrack
       sections [{
         type=MovieScene2DTransformSection
         end_frame=35
         start_frame=0
         channels [
          Translation.X {type=MovieSceneScriptingFloatChannel, const [272,2]}
          Translation.Y {type=MovieSceneScriptingFloatChannel, keys [[0.0003,214],[21.469,200],[35.0003,214]]}
          Angle {type=MovieSceneScriptingFloatChannel, const [15,1]}
          Scale.X {type=MovieSceneScriptingFloatChannel, const [1.5,1]}
          Scale.Y {type=MovieSceneScriptingFloatChannel, const [1.5,1]}
          Shear.X {type=MovieSceneScriptingFloatChannel, const [0,1]}
          Shear.Y {type=MovieSceneScriptingFloatChannel, const [0,1]}
         ]
        }]
      }]
    }
   ]}
 }
 FloatReward1 {
  binding_count=1
  length_seconds=1.75
  object=&self:FloatReward1
  sequence {bindings [Image_Reward1 {
     id=#4
     possessed_class=Image
     tracks [{
       type=MovieScene2DTransformTrack
       sections [{
         type=MovieScene2DTransformSection
         end_frame=35
         start_frame=-1
         channels [
          Translation.X {type=MovieSceneScriptingFloatChannel, const [240,2]}
          Translation.Y {type=MovieSceneScriptingFloatChannel, keys [[-0.0003,214],[21.0823,200],[35,214]]}
          Angle {type=MovieSceneScriptingFloatChannel, const [15,1]}
          Scale.X {type=MovieSceneScriptingFloatChannel, const [1.5,1]}
          Scale.Y {type=MovieSceneScriptingFloatChannel, const [1.5,1]}
          Shear.X {type=MovieSceneScriptingFloatChannel, const [0,1]}
          Shear.Y {type=MovieSceneScriptingFloatChannel, const [0,1]}
         ]
        }]
      }]
    }]}
 }
 FloatReward2 {
  binding_count=1
  length_seconds=1.75
  object=&self:FloatReward2
  sequence {bindings [Image_Reward2 {
     id=#3
     possessed_class=Image
     tracks [{
       type=MovieScene2DTransformTrack
       sections [{
         type=MovieScene2DTransformSection
         end_frame=35
         start_frame=-1
         channels [
          Translation.X {type=MovieSceneScriptingFloatChannel, const [-100,2]}
          Translation.Y {type=MovieSceneScriptingFloatChannel, keys [[-0.0003,168],[17,154],[35,168]]}
          Angle {type=MovieSceneScriptingFloatChannel, const [5,1]}
          Scale.X {type=MovieSceneScriptingFloatChannel, const [1.5,2]}
          Scale.Y {type=MovieSceneScriptingFloatChannel, const [1.5,2]}
          Shear.X {type=MovieSceneScriptingFloatChannel, const [0,1]}
          Shear.Y {type=MovieSceneScriptingFloatChannel, const [0,1]}
         ]
        }]
      }]
    }]}
 }
 FloatReward3 {
  binding_count=1
  length_seconds=1.75
  object=&self:FloatReward3
  sequence {bindings [Image_Reward3 {
     id=#2
     possessed_class=Image
     tracks [{
       type=MovieScene2DTransformTrack
       sections [{
         type=MovieScene2DTransformSection
         end_frame=35
         start_frame=-1
         channels [
          Translation.X {type=MovieSceneScriptingFloatChannel, const [-307,1]}
          Translation.Y {type=MovieSceneScriptingFloatChannel, keys [[-0.0003,-154],[12.9803,-168],[35,-154]]}
          Angle {type=MovieSceneScriptingFloatChannel, const [-5,1]}
          Scale.X {type=MovieSceneScriptingFloatChannel, const [1.5,1]}
          Scale.Y {type=MovieSceneScriptingFloatChannel, const [1.5,1]}
          Shear.X {type=MovieSceneScriptingFloatChannel, const [0,1]}
          Shear.Y {type=MovieSceneScriptingFloatChannel, const [0,1]}
         ]
        }]
      }]
    }]}
 }
 FloatReward4 {
  binding_count=1
  length_seconds=1.75
  object=&self:FloatReward4
  sequence {bindings [Image_Reward4 {
     id=#5
     possessed_class=Image
     tracks [{
       type=MovieScene2DTransformTrack
       sections [{
         type=MovieScene2DTransformSection
         end_frame=35
         start_frame=0
         channels [
          Translation.X {type=MovieSceneScriptingFloatChannel, const [78,2]}
          Translation.Y {type=MovieSceneScriptingFloatChannel, keys [[0,-95],[17,-109],[35,-95]]}
          Angle {type=MovieSceneScriptingFloatChannel, const [3,1]}
          Scale.X {type=MovieSceneScriptingFloatChannel, const [1.3,1]}
          Scale.Y {type=MovieSceneScriptingFloatChannel, const [1.3,1]}
          Shear.X {type=MovieSceneScriptingFloatChannel, const [0,1]}
          Shear.Y {type=MovieSceneScriptingFloatChannel, const [0,1]}
         ]
        }]
      }]
    }]}
 }
]
graphs [EventGraph {nodes [
   Event Construct()
   > PlayAnimation(InAnimation=Get FloatReward1() -> (FloatReward1:Widget Animation Object Reference), NumLoopsToPlay=0)
   > PlayAnimation(InAnimation=Get FloatReward2() -> (FloatReward2:Widget Animation Object Reference), NumLoopsToPlay=0)
   > PlayAnimation(InAnimation=Get FloatReward3() -> (FloatReward3:Widget Animation Object Reference), NumLoopsToPlay=0)
   > PlayAnimation(InAnimation=Get FloatReward4() -> (FloatReward4:Widget Animation Object Reference), NumLoopsToPlay=0)
   Get FloatAnim() -> (FloatAnim:Widget Animation Object Reference)
  ]}]
variables []
widgets {
 info {inheritedWidgetCount=0, parentClass=&UserWidget, rootWidgetClass=&SizeBox, widgetCount=10}
 widgets [
  SizeBox_0 {widgetClassPath=&SizeBox, widget=SizeBox_0, properties {bOverride_WidthOverride=true, heightOverride=940, widthOverride=847}}
  Overlay_0 {widgetClassPath=&Overlay, parent=SizeBox_0, widget=Overlay_0, slot {class=SizeBoxSlot}}
  Image_BackgroundGlow {
   widgetClassPath=&Image
   bIsVariable=true
   parent=Overlay_0
   widget=Image_BackgroundGlow
   properties {
    visibility=SelfHitTestInvisible
    brush {
     bIsDynamicallyLoaded=false
     drawAs=Image
     imageType=FullColor
     mirroring=NoMirror
     resourceName=None
     resourceObject=/SparksPassS05/Items/QuestRewards/BonusGoals/Materials/MI_UI_Sparks_QuestRewards_Glow_BonusGoals
     tiling=NoTile
     imageSize=<1025 1025>
     margin=<ltrb 0 0 0 0>
     outlineSettings {
      bUseBrushTransparency=false
      roundingType=HalfHeightRadius
      width=0
      color {colorUseRule=UseColor_Specified, specifiedColor=<rgba 0 0 0 0>}
      cornerRadii=<0 0 0 0>
     }
     tintColor {colorUseRule=UseColor_Specified, specifiedColor=<rgba 1 1 1 1>}
     uVRegion {bIsValid=false, max=<0 0>, min=<0 0>}
    }
    renderTransform {angle=0, scale=<1.75 2.1>, shear=<0 0>, translation=<49 0>}
   }
   slot {class=OverlaySlot, properties {horizontalAlignment=HAlign_Center, verticalAlignment=VAlign_Center}}
  }
  Image_ForegroundGlow {
   widgetClassPath=&Image
   bIsVariable=true
   parent=Overlay_0
   widget=Image_ForegroundGlow
   properties {
    visibility=SelfHitTestInvisible
    brush {
     bIsDynamicallyLoaded=false
     drawAs=Image
     imageType=FullColor
     mirroring=NoMirror
     resourceName=None
     resourceObject=/SparksPassS05/Items/QuestRewards/BonusGoals/Materials/MI_UI_Sparks_QuestRewards_ForeGlow_BonusGoals
     tiling=NoTile
     imageSize=<1024 1024>
     margin=<ltrb 0 0 0 0>
     outlineSettings {
      bUseBrushTransparency=false
      roundingType=HalfHeightRadius
      width=0
      color {colorUseRule=UseColor_Specified, specifiedColor=<rgba 0 0 0 0>}
      cornerRadii=<0 0 0 0>
     }
     tintColor {colorUseRule=UseColor_Specified, specifiedColor=<rgba 1 1 1 1>}
     uVRegion {bIsValid=false, max=<0 0>, min=<0 0>}
    }
    colorAndOpacity=<rgba 1 1 1 0.21>
    renderTransform {angle=0, scale=<1.2 0.95>, shear=<0 0>, translation=<21 -190>}
   }
   slot {class=OverlaySlot, properties {horizontalAlignment=HAlign_Right, verticalAlignment=VAlign_Center, padding=<ltrb 0 0 -50 -200>}}
  }
  Image_Reward1 {
   widgetClassPath=&Image
   bIsVariable=true
   parent=Overlay_0
   widget=Image_Reward1
   properties {
    visibility=HitTestInvisible
    brush {
     bIsDynamicallyLoaded=false
     drawAs=Image
     imageType=FullColor
     mirroring=NoMirror
     resourceName=None
     resourceObject=/SparksPassS05/Items/QuestRewards/BonusGoals/Materials/MI_UI_Sparks_QuestRewards_BonusGoals_Reward07
     tiling=NoTile
     imageSize=<200 200>
     margin=<ltrb 0 0 0 0>
     outlineSettings {
      bUseBrushTransparency=false
      roundingType=HalfHeightRadius
      width=0
      color {colorUseRule=UseColor_Specified, specifiedColor=<rgba 0 0 0 0>}
      cornerRadii=<0 0 0 0>
     }
     tintColor {colorUseRule=UseColor_Specified, specifiedColor=<rgba 1 1 1 1>}
     uVRegion {bIsValid=false, max=<0 0>, min=<0 0>}
    }
    renderTransform {angle=3, scale=<2 2>, shear=<0 0>, translation=<-279 0>}
    renderTransformPivot=<0.4 0.5>
   }
   slot {class=OverlaySlot, properties {horizontalAlignment=HAlign_Center, verticalAlignment=VAlign_Center}}
  }
  Image_Reward2 {
   widgetClassPath=&Image
   bIsVariable=true
   parent=Overlay_0
   widget=Image_Reward2
   properties {
    visibility=HitTestInvisible
    brush {
     bIsDynamicallyLoaded=false
     drawAs=Image
     imageType=FullColor
     mirroring=NoMirror
     resourceName=None
     resourceObject=/SparksPassS05/Items/QuestRewards/BonusGoals/Materials/MI_UI_Sparks_QuestRewards_BonusGoals_Reward04
     tiling=NoTile
     imageSize=<200 200>
     margin=<ltrb 0 0 0 0>
     outlineSettings {
      bUseBrushTransparency=false
      roundingType=HalfHeightRadius
      width=0
      color {colorUseRule=UseColor_Specified, specifiedColor=<rgba 0 0 0 0>}
      cornerRadii=<0 0 0 0>
     }
     tintColor {colorUseRule=UseColor_Specified, specifiedColor=<rgba 1 1 1 1>}
     uVRegion {bIsValid=false, max=<0 0>, min=<0 0>}
    }
    renderTransform {angle=3, scale=<2 2>, shear=<0 0>, translation=<0 0>}
    renderTransformPivot=<0.4 0.5>
   }
   slot {class=OverlaySlot, properties {horizontalAlignment=HAlign_Center, verticalAlignment=VAlign_Center}}
  }
  Image_Reward3 {
   widgetClassPath=&Image
   bIsVariable=true
   parent=Overlay_0
   widget=Image_Reward3
   properties {
    visibility=HitTestInvisible
    brush {
     bIsDynamicallyLoaded=false
     drawAs=Image
     imageType=FullColor
     mirroring=NoMirror
     resourceName=None
     resourceObject=/SparksPassS05/Items/QuestRewards/BonusGoals/Materials/MI_UI_Sparks_QuestRewards_BonusGoals_Reward06
     tiling=NoTile
     imageSize=<200 200>
     margin=<ltrb 0 0 0 0>
     outlineSettings {
      bUseBrushTransparency=false
      roundingType=HalfHeightRadius
      width=0
      color {colorUseRule=UseColor_Specified, specifiedColor=<rgba 0 0 0 0>}
      cornerRadii=<0 0 0 0>
     }
     tintColor {colorUseRule=UseColor_Specified, specifiedColor=<rgba 1 1 1 1>}
     uVRegion {bIsValid=false, max=<0 0>, min=<0 0>}
    }
    renderTransform {angle=3, scale=<2 2>, shear=<0 0>, translation=<241 0>}
    renderTransformPivot=<0.4 0.5>
   }
   slot {class=OverlaySlot, properties {horizontalAlignment=HAlign_Center, verticalAlignment=VAlign_Center}}
  }
  Image_Reward4 {
   widgetClassPath=&Image
   bIsVariable=true
   parent=Overlay_0
   widget=Image_Reward4
   properties {
    visibility=HitTestInvisible
    brush {
     bIsDynamicallyLoaded=false
     drawAs=Image
     imageType=FullColor
     mirroring=NoMirror
     resourceName=None
     resourceObject=/SparksPassS05/Items/QuestRewards/BonusGoals/Materials/MI_UI_Sparks_QuestRewards_BonusGoals_Reward05
     tiling=NoTile
     imageSize=<200 200>
     margin=<ltrb 0 0 0 0>
     outlineSettings {
      bUseBrushTransparency=false
      roundingType=HalfHeightRadius
      width=0
      color {colorUseRule=UseColor_Specified, specifiedColor=<rgba 0 0 0 0>}
      cornerRadii=<0 0 0 0>
     }
     tintColor {colorUseRule=UseColor_Specified, specifiedColor=<rgba 1 1 1 1>}
     uVRegion {bIsValid=false, max=<0 0>, min=<0 0>}
    }
    renderTransform {angle=3, scale=<2 2>, shear=<0 0>, translation=<241 0>}
    renderTransformPivot=<0.4 0.5>
   }
   slot {class=OverlaySlot, properties {horizontalAlignment=HAlign_Center, verticalAlignment=VAlign_Center}}
  }
  Border_53 {
   widgetClassPath=&Border
   parent=Overlay_0
   widget=Border_53
   properties {
    visibility=Hidden
    background {
     bIsDynamicallyLoaded=false
     drawAs=Image
     imageType=FullColor
     mirroring=NoMirror
     resourceName=None
     resourceObject=/SparksPassS05/Items/QuestRewards/Materials/MI_UI_Sparks_QuestRewards_InfoBack_FestivalPass
     tiling=NoTile
     imageSize=<32 32>
     margin=<ltrb 0 0 0 0>
     outlineSettings {
      bUseBrushTransparency=false
      roundingType=HalfHeightRadius
      width=0
      color {colorUseRule=UseColor_Specified, specifiedColor=<rgba 0 0 0 0>}
      cornerRadii=<0 0 0 0>
     }
     tintColor {colorUseRule=UseColor_Specified, specifiedColor=<rgba 1 1 1 1>}
     uVRegion {bIsValid=false, max=<0 0>, min=<0 0>}
    }
    padding=<ltrb 50 20 50 20>
   }
   slot {class=OverlaySlot, properties {horizontalAlignment=HAlign_Center, verticalAlignment=VAlign_Bottom, padding=<ltrb 40 40 20 40>}}
  }
  RichText_QRCode {
   widgetClassPath=&CommonRichTextBlock
   bIsVariable=true
   parent=Border_53
   widget=RichText_QRCode
   properties {
    autoWrapText=true
    text="Complete <Highlight>Axion Sentinel</> quests to earn rewards."
    textStyleSet=/SparksPassS05/Items/QuestRewards/DT_UI_RichTextStyles_Sparks_QuestRewards_FestivalPass
    decoratorClasses [/Game/UI/Sidebar/RebootRally/BP_UI_RebootRallyRichImageDecorator.BP_UI_RebootRallyRichImageDecorator_C]
    defaultTextStyleOverride {
     overflowPolicy=Clip
     transformPolicy=None
     colorAndOpacity {colorUseRule=UseColor_Specified, specifiedColor=<rgba 1 1 1 1>}
     font {
      bForceMonospaced=false
      bMaterialIsStencil=false
      fontMaterial=null
      fontObject=&BurbankSmall
      letterSpacing=0
      monospacedWidth=1
      size=19
      skewAmount=0
      typefaceFontName=Bold
      outlineSettings {
       bApplyOutlineToDropShadows=false
       bMiteredCorners=false
       bSeparateFillAlpha=false
       outlineMaterial=null
       outlineSize=0
       outlineColor=<rgba 0 0 0 1>
      }
     }
     highlightColor {colorUseRule=UseColor_Specified, specifiedColor=<rgba 0 0 0 1>}
     highlightShape {
      bIsDynamicallyLoaded=false
      drawAs=Image
      imageType=NoImage
      mirroring=NoMirror
      resourceName=None
      resourceObject=null
      tiling=NoTile
      imageSize=<32 32>
      margin=<ltrb 0 0 0 0>
      outlineSettings {
       bUseBrushTransparency=false
       roundingType=HalfHeightRadius
       width=0
       color {colorUseRule=UseColor_Specified, specifiedColor=<rgba 0 0 0 0>}
       cornerRadii=<0 0 0 0>
      }
      tintColor {colorUseRule=UseColor_Specified, specifiedColor=<rgba 1 1 1 1>}
      uVRegion {bIsValid=false, max=<0 0>, min=<0 0>}
     }
     selectedBackgroundColor {colorUseRule=UseColor_Foreground, specifiedColor=<rgba 1 0 1 1>}
     shadowColorAndOpacity=<rgba 0 0 0 1>
     shadowOffset=<0 0>
     strikeBrush {
      bIsDynamicallyLoaded=false
      drawAs=Image
      imageType=NoImage
      mirroring=NoMirror
      resourceName=None
      resourceObject=null
      tiling=NoTile
      imageSize=<32 32>
      margin=<ltrb 0 0 0 0>
      outlineSettings {
       bUseBrushTransparency=false
       roundingType=HalfHeightRadius
       width=0
       color {colorUseRule=UseColor_Specified, specifiedColor=<rgba 0 0 0 0>}
       cornerRadii=<0 0 0 0>
      }
      tintColor {colorUseRule=UseColor_Specified, specifiedColor=<rgba 1 1 1 1>}
      uVRegion {bIsValid=false, max=<0 0>, min=<0 0>}
     }
     underlineBrush {
      bIsDynamicallyLoaded=false
      drawAs=Image
      imageType=NoImage
      mirroring=NoMirror
      resourceName=None
      resourceObject=null
      tiling=NoTile
      imageSize=<32 32>
      margin=<ltrb 0 0 0 0>
      outlineSettings {
       bUseBrushTransparency=false
       roundingType=HalfHeightRadius
       width=0
       color {colorUseRule=UseColor_Specified, specifiedColor=<rgba 0 0 0 0>}
       cornerRadii=<0 0 0 0>
      }
      tintColor {colorUseRule=UseColor_Specified, specifiedColor=<rgba 1 1 1 1>}
      uVRegion {bIsValid=false, max=<0 0>, min=<0 0>}
     }
    }
   }
   slot {class=BorderSlot, properties {padding=<ltrb 50 20 50 20>}}
  }
 ]
}
